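; Blueprint actor for a vault keycard lock in Heist bases, extending a parent keycard lock class.
; On BeginPlay it runs the parent logic, then checks a hotfixable StartOpened scalable float and an IsTakeover flag.
; When both are true, it sets its Lock state to a specific enum value, likely opened so takeover vaults start unlocked.
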(event EventBeginPlay
  (|CallParentFunction)
  (if (and (ScalableFloat|GetValueasBool (Variables|Default|GetStartOpened_Hotfix)) (Variables|Default|GetIsTakeover))
    (Variables|Default|SetLock "NewEnumerator2")))

(fn ConstructionScript
  (|CallParentFunction))
