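; Eclipse vault screen wall Blueprint, a child class that extends its parent's alarm visual update with per-instance screen UV settings.
; UpdateAlarmVisuals calls the parent, then packs the Tile_UV_X and Offset_UV_X 2D vectors into one Vector4 written to the static mesh's custom primitive data at index 0.
; It returns the parent's active result unchanged; the construction script only calls the parent.
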(fn UpdateAlarmVisuals
  (params (ForceOff :type Boolean) (NewState :type Boolean))
  (bind (_tile_uv_x _tile_uv_y) (Math|Vector2D|BreakVector2D (Variables|GetTile_UV_X)))
  (bind (_offset_uv_x _offset_uv_y) (Math|Vector2D|BreakVector2D (Variables|GetOffset_UV_X)))
  (bind _isactive (|CallParentFunction ForceOff NewState))
  (Utilities|FlowControl|Sequence
    (:then_0
      (Rendering|Material|SetCustomPrimitiveDataVector4 (Variables|Debug|GetStaticMeshComponent) 0 (Math|Vector4|MakeVector4 _tile_uv_x _tile_uv_y _offset_uv_x _offset_uv_y)))
    (:then_1
      (return _isactive))))

(fn ConstructionScript
  (|CallParentFunction))
