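; Heist base vault door that shows its closed or heisted state through a map icon and a material change.
; On begin play the authority optionally spawns a map icon actor with the closed icon, and end play destroys it.
; The VaultHasOpened event sets VaultOpened, and the authority swaps to the heisted icon.
; VaultOpened's OnRep writes game time into the door's custom primitive data for time-based material effects, only where cosmetics can run.
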(event EventBeginPlay
  (|CallParentFunction)
  (Default|BindEventtoVaultHasOpened (delegate Custom|VaultHasOpened_Event))
  (Networking|SwitchHasAuthority
    (:Authority
      (if (Variables|Default|GetShowMapIcon)
        (bind _returnvalue (Game|SpawnActorfromClass "/HeistBasesGameplay/Items/B_HeistBaseMapIconActor.B_HeistBaseMapIconActor_C" (Math|Conversions|ToTransform(Vector) (Transformation|GetActorLocation)) "AlwaysSpawn" "MultiplyWithRoot" 0 0))
        (bind _output_get (Variables|Default|SetMapIconActor _returnvalue))
        (Utilities|IsValid _output_get
          (:Is%20Valid
            (Class|BHeistBaseMapIconActor|SetMapIcon _output_get (Variables|Default|GetClosedMapIcon))
            (Class|BHeistBaseMapIconActor|SetIconScale _output_get "(X=1.500000,Y=1.500000)")))))))

(event Custom|VaultHasOpened_Event
  (bind _mapiconactor (Variables|Default|GetMapIconActor))
  (Variables|Default|SetVaultOpened true)
  (Networking|SwitchHasAuthority
    (:Authority
      (if (Variables|Default|GetShowMapIcon)
        (Utilities|IsValid _mapiconactor
          (:Is%20Valid
            (Class|BHeistBaseMapIconActor|SetMapIcon _mapiconactor (Variables|Default|GetHeistedMapIcon))))))))

(event EventEndPlay
  (params (EndPlayReason :type EEndPlayReason%20Enum :default "Destroyed"))
  (bind _mapiconactor (Variables|Default|GetMapIconActor))
  (|CallParentFunction EndPlayReason)
  (Utilities|IsValid _mapiconactor
    (:Is%20Valid
      (Actor|DestroyActor _mapiconactor))))

(fn OnRep_VaultOpened
  (Networking|CanExecuteCosmeticEvents
    (:True
      (Rendering|Material|SetDefaultCustomPrimitiveDataFloat (Variables|Default|GetVaultDoor) 0 (Utilities|Time|GetGameTimeinSeconds)))))

(fn ConstructionScript
  (|CallParentFunction))
